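# One-shot mono MetaSound source for a playable instrument with attenuation, playing MIDI-driven notes and an end flourish.
# A change to Instrument picks a Fusion patch and flourish wave from two arrays by index.
# PlayNote sends CurrentNote through a MIDI note generator into a Fusion sampler, mixed with the PlayEndFlourish wave player.
# OnFinished fires shortly after a note delay or flourish completion, and each new note resets that final delay.
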MetaSoundSource /Hera/Audio/PlayableMusic/MSS_PlayableInstrument
inputs [
 CurrentNote {type=Int32}
 Instrument {type=Int32}
 PlayEndFlourish {type=Trigger}
 PlayNote {type=Trigger}
 UE.Source.OnPlay {type=Trigger}
]
interfaces [UE.OutputFormat.Mono:1.0, UE.Source.OneShot:1.0, UE.Source:1.1]
outputs [
 "UE.OutputFormat.Mono.Audio:0" {type=Audio}
 UE.Source.OneShot.OnFinished {type=Trigger}
]
pages [Default {
  input_defaults {CurrentNote=0, Instrument=0, PlayEndFlourish=null, PlayNote=null, UE.Source.OnPlay=null}
  nodes [
   TriggerOnValueChange_0 = UE.Trigger On Value Change:Int32(Value=^Instrument)
   n1 = VariableAccessor.FusionPatchAsset:Array(Variable=InitVariable.FusionPatchAsset:Array().Variable)
   n2 = InitVariable.WaveAsset:Array()
   WavePlayer_0 = UE.Wave Player:Mono(
    Play=^PlayEndFlourish
    Wave Asset=Array.Get:WaveAsset:Array(Index=^Instrument, Trigger=@TriggerOnValueChange_0.Trigger, Array=VariableAccessor.WaveAsset:Array(Variable=@n2.Variable).Value).Element
   ) {also {version [1,1]}}
   AudioMixerAudioMixerMono2_0 = "AudioMixer.Audio Mixer (Mono, 2)"(
    In 0=HarmonixNodes.FusionSampler(
     Patch=Array.Get:FusionPatchAsset:Array(Index=^Instrument, Trigger=@TriggerOnValueChange_0.Trigger, Array=@n1.Value).Element
     MIDI Stream=HarmonixNodes.MidiNoteGenerator(MIDI Note=^CurrentNote, Trigger=^PlayNote) {also {defaults {"Note Off Delay (Seconds)"=3}}}.MIDI Stream
    ) {also {version [0,1]}}.Audio Mono
    In 1=@WavePlayer_0.Out Mono
   )
   TriggerDelay_1 = UE.Trigger Delay(In=^PlayNote) {also {defaults {Delay Time=3}}}
   TriggerAnyTriggerAny2_0 = "TriggerAny.Trigger Any (2)"(In 1=@TriggerDelay_1.Out, In 0=@WavePlayer_0.On Finished)
   TriggerDelay_0 = UE.Trigger Delay(Reset=^PlayNote, In=@TriggerAnyTriggerAny2_0.Out)
  ]
  links [{dst="@UE.OutputFormat.Mono.Audio:0", src=AudioMixerAudioMixerMono2_0.Out}, {dst="@UE.Source.OneShot.OnFinished", src=TriggerDelay_0.Out}]
 }]
properties {
 assetImportData=&self:AssetImportData
 attenuationSettings=&/ATT_PlayableInstrument
 maxDistance=4000
 sampleRate=48000
 soundClassObject=/Game/Sounds/FortSoundClassesAndModes/Fort_Music_Gameplay
 volume=3
 concurrencySet [&/CONC_PlayableInstrument_Owner, &/CONC_PlayableInstrument]
}
